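; Blueprint color palette for the Mojo color shift in the stage show's FX manager, storing a flat array of linear colors.
; GetColor treats the array as consecutive groups of three, returning the trio that starts at the given index times three.
; An index whose trio runs past the array end prints an on-screen out-of-range message instead.
; GetColors and UpdateColors read and replace the whole array.
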(fn GetColor
  (params (Color%20Index :type Integer))
  (bind _colors (Variables|Default|GetColors))
  (bind _returnvalue (Utilities|Array|Length _colors))
  (bind _returnvalue_2 (* Color%20Index 3))
  (if (< (+ _returnvalue_2 2) _returnvalue)
    (return (Utilities|Array|Get(acopy) (Variables|Default|GetColors) _returnvalue_2) (Utilities|Array|Get(acopy) (Variables|Default|GetColors) (+ _returnvalue_2 1)) (Utilities|Array|Get(acopy) (Variables|Default|GetColors) (+ _returnvalue_2 2)))
    (else
      (Development|PrintText 0 (Utilities|Text|FormatText "NSLOCTEXT(\"[01FE65F845BCF2994C91451681B2C008]\", \"B3B42DCD42CD8CEE6C92D29D2FF96BC6\", \"Collor palette index out of range: {index}/{lenght}. Raw index: {raw} \")" _returnvalue_2 (- _returnvalue 2) Color%20Index) true true "(R=1.000000,G=0.678748,B=0.166667,A=1.000000)" 1e+01))))

(fn GetColors
  (return (Variables|Default|GetColors)))

(fn UpdateColors
  (params (Colors :type Array%20of%20Linear%20Color%20Structures))
  (Variables|Default|SetColors Colors))
