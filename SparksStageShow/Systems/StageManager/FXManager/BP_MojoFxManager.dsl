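; BP_MojoFxManager is a stage show manager that plays choreographed FX events across its set of stage actors.
; StageShowEvent copies data asset settings, rejecting events while busy, and plays one-shot or looping events by timer or StageManager beat/bar callbacks.
; PlayNextFX re-arms a timer to step through fire-all, indices, linear, alternating, outside-in or random patterns, messaging actors via StageShowFXInterface.
; An autonomous loop mode ignores incoming events and cycles its own validated event list.
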(event Custom|LoadAutonomusParticlesAssets
  ;[1 1 1] Load the Niagara System soft reference if we gonna use autonomous loop
  (for (_array_element _array_index) (Variables|AutonomousLoop|GetAutonomousLoopEvents)
    (Utilities|IsValid _array_element
      (:Is%20Valid
        (if (Utilities|IsValidSoftObjectReference (Variables|/SparksStageShow/Systems/StageManager/FXManager/MojoFxManagerEvent_DataAsset.MojoFxManagerEvent_DataAsset_C|GetNiagaraSystemOverride _array_element))
          (bind _object (Utilities|AsyncLoadAsset (Variables|/SparksStageShow/Systems/StageManager/FXManager/MojoFxManagerEvent_DataAsset.MojoFxManagerEvent_DataAsset_C|GetNiagaraSystemOverride _array_element))
            (:Completed
              (Variables|Default|SetTempLoaded _object)))))
      (:Is%20Not%20Valid
        (Development|PrintText (Utilities|Text|FormatText "NSLOCTEXT(\"[9F215592403329545C16AD81FBC8B0E5]\", \"7AA99E344BAA456A3779F29719AFA051\", \"[FxManager] {BPName} has an invalid event on the AutonomousLoopEvents array. Index: {index}.\")" (Utilities|GetDisplayName self) _array_index) true true "(R=1.000000,G=0.123958,B=0.268309,A=1.000000)"))))
  (Variables|Default|SetTempLoaded 0))

(event EventBeginPlay
  ;[0.1263 0.4167 0.07326] Begin Play. It starts ambient FX using seconds as timing as soon game starts or waits for the Intro event
  (Networking|CanExecuteCosmeticEvents
    (:True
      (if (Variables|AutonomousLoop|GetAutonomousLoop)
        (CallFunction|LoadAutonomusParticlesAssets)
        (CallFunction|SetUpEvents)))))

(event Custom|StageShowEvent
  (params (FXManagerEvent :type Mojo%20Fx%20Manager%20Event%20Data%20Asset%20Object%20Reference))
  (bind _stagemanager (Variables|AutonomousLoop|GetStageManager))
  (bind _currentfxevent (Variables|Globals|GetCurrentFxEvent))
  ;[0.3846 0.4252 1] Stage Show Event: Executes a single FxManager event. Used by the StageManager sequences
  ;;[1 1 1] If auto loop is enabled, this FxManager will ignore any arriving event from outside
  (if (not (Variables|AutonomousLoop|GetAutonomousLoop))
    ;;[1 1 1] Check if it's a "Stop" event
    (if (not (Variables|/SparksStageShow/Systems/StageManager/FXManager/MojoFxManagerEvent_DataAsset.MojoFxManagerEvent_DataAsset_C|GetStopFXManager FXManagerEvent))
      ;;[0.5478 0.9901 1] This whole block handles what to do if an earlier event is still playing when the new event arrived
      (if (Variables|Globals|GetisPlayingLoopingEvent)
        ;;;[1 1 1] If it's currently looping an event, halt the loop before starting the next event
        (AmbientFX|StopLoopingEvent)
        (elif (not (Variables|Globals|GetIsBusy))
          _
          (else
            ;;[1 0.8253 0.4362] If we are executing an earlier event, we'll ignore the event that arrived and notify the sequence to proceed to the next event
            (Default|CallEventFinished self)
            (Utilities|IsValid _stagemanager
              (:Is%20Valid
                (if (Variables|/SparksStageShow/Systems/StageManager/BP_MojoStageManagerBase.BP_MojoStageManagerBase_C|GetEnableDebugLogs _stagemanager)
                  (Development|PrintText (Utilities|Text|FormatText "NSLOCTEXT(\"[9F215592403329545C16AD81FBC8B0E5]\", \"5210AA2D4DDBEDA6F50D5A9281D779CA\", \"[FxManager] {EventName} event was rejected because\\r\\nthe FxManager is busy.\")" (Utilities|GetDisplayName FXManagerEvent)))))))))
      ;;[1 1 1] Here we set the new event and check if it's a looping event
      (Variables|Globals|SetCurrentFxEvent FXManagerEvent)
      (if (Variables|/SparksStageShow/Systems/StageManager/FXManager/MojoFxManagerEvent_DataAsset.MojoFxManagerEvent_DataAsset_C|GetLoopingEvent _currentfxevent)
        ;;[1 1 1] If it's a looping event, we set it to repeat by a timer or to repeat each beat/measure.
        ;; This event will keep repeting until a new event arrives.
        (Variables|Globals|SetisPlayingLoopingEvent true)
        (if (Utilities|Enum|Equal(Enum) (Variables|/SparksStageShow/Systems/StageManager/FXManager/MojoFxManagerEvent_DataAsset.MojoFxManagerEvent_DataAsset_C|GetFrequencyType _currentfxevent) "None")
          (CallFunction|ListentoBeatEvent false)
          (AmbientFX|StartLoopingEventbyTimer)
          (elif (Variables|Default|GetAssignedToBeat)
            _
            (else
              (CallFunction|ListentoBeatEvent true))))
        (Default|CallEventFinished self)
        (else
          ;;;[1 1 1] If it's not a looping event, just fire the Fx once
          (CallFunction|SetGlobalVars _currentfxevent)
          (Variables|Globals|SetIsBusy true)
          (CallFunction|ListentoBeatEvent false)
          (Default|PlayNextFX)))
      (else
        ;;;[1 1 1] If it's a "stop" event, imediatelly interrupt the animation, resets everything and notify the sequence/spawned events to proceed to the next sequence step
        (CallFunction|StopEventandFinish)))))

(event Custom|SetUpEvents
  ;[0.2865 0.2865 0.2865] Song Event Setup: Listens to gameplay events and starts autonomous loop if enabled. If it's on beat or measure timing mode, switched to that when song starts playing.
  ;;[1 1 1] If AutonomousLoop is true and we have a array of valid events, proceed
  (if (Variables|AutonomousLoop|GetAutonomousLoop)
    (bind _isvalid (AmbientFX|ValidateAutonomousLoopEvents))
    (if (Variables|AutonomousLoop|GetIgnoreLoopConditions)
      (CallFunction|OnEventCalled "None")
      (elif (and _isvalid (!= (Variables|AutonomousLoop|GetStartLoopState) (Variables|AutonomousLoop|GetStopLoopState)))
        (Default|BindEventtoOnEventCalled (CallFunction|GetStageManager) (EventDispatchers|CreateEvent :!event Custom|OnEventCalled))))))

(event Custom|OnEventCalled
  (params (EventName :type Name))
  (bind (_stagemanager _isvalid) (CallFunction|GetStageManager))
  (if (not (Variables|AutonomousLoop|GetIgnoreLoopConditions))
    ;[1 1 1] If we're on start state, start the auto loop using the frequency type defined by the first event
    (Utilities|FlowControl|Sequence
      (:then_0
        (if (== (Variables|AutonomousLoop|GetStartLoopState) EventName)
          _))
      (:then_1
        ;[1 1 1] If we reached the stop state, interrupt the auto loop
        (if (== EventName (Variables|AutonomousLoop|GetStopLoopState))
          (AmbientFX|StopLoopingEvent)))))
  (bind _output_get (Variables|Globals|SetCurrentFxEvent (Utilities|Array|Get(acopy) (Variables|AutonomousLoop|GetValidAutonomousLoopEvents))))
  (Utilities|FlowControl|Delay (Variables|/SparksStageShow/Systems/StageManager/FXManager/MojoFxManagerEvent_DataAsset.MojoFxManagerEvent_DataAsset_C|GetStartDelay _output_get))
  (switch Utilities|FlowControl|Switch|SwitchonEMidiClockSubdivisionQuantization (Variables|/SparksStageShow/Systems/StageManager/FXManager/MojoFxManagerEvent_DataAsset.MojoFxManagerEvent_DataAsset_C|GetFrequencyType _output_get)
    (:Bar
      _)
    (:Beat
      _)
    (:ThirtySecondNote
      _)
    (:SixteenthNote
      _)
    (:EighthNote
      _)
    (:QuarterNote
      _)
    (:HalfNote
      _)
    (:WholeNote
      _)
    (:DottedSixteenthNote
      _)
    (:DottedEighthNote
      _)
    (:DottedQuarterNote
      _)
    (:DottedHalfNote
      _)
    (:DottedWholeNote
      _)
    (:SixteenthNoteTriplet
      _)
    (:EighthNoteTriplet
      _)
    (:QuarterNoteTriplet
      _)
    (:HalfNoteTriplet
      _)
    (:None
      (AmbientFX|StartLoopingEventbyTimer :!label _label)
      (CallFunction|ListentoBeatEvent false)))
  (if _isvalid
    (if (not (Variables|Default|GetAssignedToBeat))
      (CallFunction|ListentoBeatEvent true))
    (else
      (goto _label))))

(event Custom|StartListeningBeatEvents
  (params (BeatInBar :type Integer) (MeasureNumber :type Integer) (IsBeating :type Boolean :default true))
  ;[1 1 1] OnBeat and OnMeasure events setup
  (if IsBeating
    (if (Utilities|Time|IsValidTimerHandle (Variables|AutonomousLoop|GetAutonomousLoopTimer))
      (AmbientFX|StopLoopingEvent))
    (switch Utilities|FlowControl|Switch|SwitchonEMidiClockSubdivisionQuantization (Variables|/SparksStageShow/Systems/StageManager/FXManager/MojoFxManagerEvent_DataAsset.MojoFxManagerEvent_DataAsset_C|GetFrequencyType (Variables|Globals|GetCurrentFxEvent))
      (:Bar
        (if (== BeatInBar 1)
          (AmbientFX|OnMeasure)))
      (:Beat
        (AmbientFX|OnBeat)))
    (else
      (AmbientFX|StartLoopingEventbyTimer))))

(fn CallColorShiftOnActors
  (params (Primary :type Linear%20Color%20Structure) (Secondary :type Linear%20Color%20Structure) (Tertiary :type Linear%20Color%20Structure) (AltPrimary :type Linear%20Color%20Structure) (AltSecondary :type Linear%20Color%20Structure) (AltTertiary :type Linear%20Color%20Structure))
  (for _array_element (Variables|Globals|GetStageShowActors)
    (Class|StageShowFXInterface|MojoStageShowColorShift(Message) _array_element Primary Secondary Tertiary AltPrimary AltSecondary AltTertiary 0.0)))

(fn FireDebugEvent
  (if (not (Variables|Globals|GetIsBusy))
    (Variables|Globals|SetIsBusy true)
    (CallFunction|SetGlobalVars (Variables|Debug|GetDebugState))
    (Default|PlayNextFX)))

(fn FireLoopingEvent
  (LocalVariables|SetFxType (Utilities|Enum|BytetoEnumEShowFXType (Math|Conversions|ToByte(Integer) (Math|Conversions|ToInteger(Byte) (Variables|/SparksStageShow/Systems/StageManager/FXManager/MojoFxManagerEvent_DataAsset.MojoFxManagerEvent_DataAsset_C|GetFxType (Variables|Globals|GetCurrentFxEvent))))))
  ;[1 1 1] If a new ambient event arrives, it'll be ignored if this manager is already executing another event.
  ;
  ; That means that if ambient frequency is too high, it'll skip some events.
  (if (not (Variables|Globals|GetIsBusy))
    ;[1 1 1] Set the manager variables based on the receive event and fire the animation
    (CallFunction|SetGlobalVars (Variables|Globals|GetCurrentFxEvent))
    (Default|PlayNextFX)))

(fn GetNextAutonomousEvent
  ;[1 1 1] Gets the next event on the autonomous event array and set as the current event to be played
  (Math|Integer|IncrementInt (Variables|AutonomousLoop|GetAutonomousLoopEventsIndex)
    (:%20%20
      (Utilities|FlowControl|Sequence
        (:then_0
          (if (>= (Variables|AutonomousLoop|GetAutonomousLoopEventsIndex) (Utilities|Array|Length (Variables|AutonomousLoop|GetValidAutonomousLoopEvents)))
            (Variables|AutonomousLoop|SetAutonomousLoopEventsIndex 0)))
        (:then_1
          (Variables|Globals|SetCurrentFxEvent (Utilities|Array|Get(acopy) (Variables|AutonomousLoop|GetValidAutonomousLoopEvents) (Variables|AutonomousLoop|GetAutonomousLoopEventsIndex))))))))

(fn GetNextFxAlternate
  (params (Index :type Integer%20%28by%20ref%29) (LoopIndex :type Integer%20%28by%20ref%29))
  (bind _alternatecount (Variables|Globals|GetAlternateCount))
  (bind _myindex (LocalVariables|GetmyIndex))
  (bind _returnvalue (Math|Integer|%25%28Integer%29 _myindex _alternatecount))
  (LocalVariables|SetmyIndex Index)
  (LocalVariables|SetmyLoopIndex LoopIndex)
  (Utilities|Array|Clear (LocalVariables|GetmyResult))
  ;[1 1 1] This animation logic only works with a value of 2. Abort if it's below that
  (if (> (Variables|Globals|GetAlternateCount) 1)
    ;[1 1 1] Checking if all anim loops were executed
    (if (< (LocalVariables|GetmyLoopIndex) (Variables|Globals|GetLoopCount))
      ;[0.03498 0.1148 0.1219] This will generate an alternating pattern that moves.
      ; For exemple, if alternate count is set to 3 and we have 9 elements, each iteration of the animation will produce the following pattern:
      ; 100100100
      ; 010010010
      ; 001001001
      (for _index (range (Utilities|Array|Length (Variables|Globals|GetStageShowActors)))
        ;;[1 1 1] Add the select actors to the results array
        (if (== (Math|Integer|%25%28Integer%29 _index _alternatecount) _returnvalue)
          (Utilities|Array|Add (LocalVariables|GetmyResult) (Utilities|Array|Get(acopy) (Variables|Globals|GetStageShowActors) _index))))
      ;;[1 1 1] A loop is completed when the pattern moved a number of times equal AlternateCount - 1
      (if (== (- (Variables|Globals|GetAlternateCount) 1) (LocalVariables|GetmyIndex))
        (Variables|SetBy-RefVar (Variables|Globals|GetLoopIndex) (+ (LocalVariables|GetmyLoopIndex) 1)))
      ;;[1 1 1] This will make the index loop whenever
      ;; it reaches the AlternateCount value
      ;; Same as (if index > AlternateCount - 1) index = 0
      (Variables|SetBy-RefVar (Variables|GetIndex) (+ _returnvalue 1))
      (else
        (Variables|SetBy-RefVar (Variables|Globals|GetLoopIndex)))))
  (return (LocalVariables|GetmyResult) (and (< (Variables|Globals|GetLoopIndex) (Variables|Globals|GetLoopCount)) (> (Variables|Globals|GetLoopIndex) 0))))

(fn GetNextFxFireAll
  (params (LoopIndex :type Integer%20%28by%20ref%29))
  (LocalVariables|SetmyLoopIndex LoopIndex)
  (Utilities|Array|Clear (LocalVariables|GetmyResult))
  (if (< (LocalVariables|GetmyLoopIndex) (Variables|Globals|GetLoopCount))
    ;[1 1 1] Just pass all actors and iterate the loop index
    (Variables|SetBy-RefVar (Variables|Globals|GetLoopIndex) (+ (LocalVariables|GetmyLoopIndex) 1))
    (LocalVariables|SetmyResult (Variables|Globals|GetStageShowActors))
    (else
      ;[1 1 1] If all loops were executed, reset the loop index and end animation
      (Variables|SetBy-RefVar (Variables|Globals|GetLoopIndex))))
  (return (LocalVariables|GetmyResult)))

(fn GetNextFxFireIndices
  (params (LoopIndex :type Integer%20%28by%20ref%29))
  (LocalVariables|SetmyLoopIndex LoopIndex)
  (Utilities|Array|Clear (LocalVariables|GetmyResult))
  (if (< (LocalVariables|GetmyLoopIndex) (Variables|Globals|GetLoopCount))
    (Variables|SetBy-RefVar (Variables|Globals|GetLoopIndex) (+ (LocalVariables|GetmyLoopIndex) 1))
    (if (Variables|Globals|GetReverse)
      ;[1 1 1] The reverse mode selects actors that are not on the indices array
      (for _index (range (Utilities|Array|Length (Variables|Globals|GetStageShowActors)))
        (if (not (Utilities|Array|ContainsItem (Variables|Globals|GetIndices) _index))
          (Utilities|Array|Add (LocalVariables|GetmyResult) (Utilities|Array|Get(acopy) (Variables|Globals|GetStageShowActors) _index))))
      (else
        ;[1 1 1] Normal mode just adds the actors on the selected indices
        (for _array_element (Variables|Globals|GetIndices)
          (Utilities|Array|Add (LocalVariables|GetmyResult) (Utilities|Array|Get(acopy) (Variables|Globals|GetStageShowActors) _array_element)))))
    (else
      ;[1 1 1] If all loops are completed, reset loop index and end the animation
      (Variables|SetBy-RefVar (Variables|Globals|GetLoopIndex))))
  (return (LocalVariables|GetmyResult) (and (< (Variables|Globals|GetLoopIndex) (Variables|Globals|GetLoopCount)) (> (Variables|Globals|GetLoopIndex) 0))))

(fn GetNextFxLinear
  (params (Index :type Integer%20%28by%20ref%29) (LoopIndex :type Integer%20%28by%20ref%29))
  (bind _myloopindex (LocalVariables|GetmyLoopIndex))
  (bind _returnvalue (+ _myloopindex 1))
  (LocalVariables|SetmyIndex Index)
  (LocalVariables|SetmyLoopIndex LoopIndex)
  (Utilities|Array|Clear (LocalVariables|GetmyResult))
  (if (not (< (LocalVariables|GetmyIndex) (Utilities|Array|Length (Variables|Globals|GetStageShowActors))))
    ;[1 1 1] Reset the actor index and count a completed loop. If we completed the set loop count,
    ; we finished this animation
    (LocalVariables|SetmyIndex 0)
    (Variables|SetBy-RefVar (Variables|Globals|GetLoopIndex) _returnvalue)
    (if (< _returnvalue (Variables|Globals|GetLoopCount))
      ;[1 1 1] Add the actor on the current index
      (Utilities|Array|Add (LocalVariables|GetmyResult) (Utilities|Array|Get(acopy) (Variables|Globals|GetStageShowActors) (Math|Integer|SelectInt (- (- (Utilities|Array|Length (Variables|Globals|GetStageShowActors)) (LocalVariables|GetmyIndex)) 1) (LocalVariables|GetmyIndex) (Variables|Globals|GetReverse))) :!label _label)
      ;[1 1 1] Increment index
      (Variables|SetBy-RefVar (Variables|GetIndex) (+ (LocalVariables|GetmyIndex) 1))
      (else
        (Variables|SetBy-RefVar (Variables|Globals|GetLoopIndex))))
    (return (LocalVariables|GetmyResult) (< (LocalVariables|GetmyIndex) (Utilities|Array|Length (Variables|Globals|GetStageShowActors)))))
  (goto _label))

(fn GetNextFxOutsideInOut
  (params (Index :type Integer%20%28by%20ref%29) (LoopIndex :type Integer%20%28by%20ref%29))
  (bind _stageshowactors (Variables|Globals|GetStageShowActors))
  (bind _returnvalue (Utilities|Array|Length _stageshowactors))
  (bind _stageshowactors_2 (Variables|Globals|GetStageShowActors))
  (bind _returnvalue_2 (Utilities|Array|Length _stageshowactors_2))
  (bind _returnvalue_3 (Math|Conversions|ToFloat(Integer) _returnvalue_2))
  (bind _reverse (Variables|Globals|GetReverse))
  (bind _myindex (LocalVariables|GetmyIndex))
  (bind _returnvalue_4 (Math|Integer|SelectInt (+ (Math|Float|Floor (/ _returnvalue_3 2.0)) (LocalVariables|GetmyIndex)) _myindex _reverse))
  (bind _returnvalue_5 (/ _returnvalue_3 2.0))
  (bind _returnvalue_6 (Math|Float|Floor _returnvalue_5))
  (bind _myloopindex (LocalVariables|GetmyLoopIndex))
  (bind _returnvalue_7 (+ _myloopindex 1))
  (bind _returnvalue_8 (- _returnvalue _returnvalue_4))
  (bind _returnvalue_9 (- _returnvalue_8 1))
  (bind _myindex_2 (LocalVariables|GetmyIndex))
  (bind _returnvalue_10 (+ _returnvalue_6 _myindex_2))
  (LocalVariables|SetmyIndex Index)
  (LocalVariables|SetmyLoopIndex LoopIndex)
  (Utilities|Array|Clear (LocalVariables|GetmyResult))
  ;[1 1 1] Are we not on or after the middle actor?
  (if (not (< (LocalVariables|GetmyIndex) (Math|Float|Ceil _returnvalue_5)))
    ;[1 1 1] If we reached the middle Actor index, a loop was completed.
    ; Reset variables and iterate the loop counetr
    (LocalVariables|SetmyIndex 0)
    (Variables|SetBy-RefVar (Variables|Globals|GetLoopIndex) _returnvalue_7)
    (if (< _returnvalue_7 (Variables|Globals|GetLoopCount))
      ;[1 1 1] Add the current index or use the index starting from the middle if reverse
      (Utilities|Array|Add (LocalVariables|GetmyResult) (Utilities|Array|Get(acopy) (Variables|Globals|GetStageShowActors) _returnvalue_4) :!label _label)
      ;[1 0.9138 0.3166] If index and mirrored index are different, we also add the mirrored actor to the results.
      ; (This prevents an actor beign added twice when we have a odd number of actors and eventually, index and mirrored index are the same when we reach the middle)
      ;
      ; Could possibly be just and ADD Unique?
      (if (!= _returnvalue_4 _returnvalue_9)
        (Utilities|Array|Add (LocalVariables|GetmyResult) (Utilities|Array|Get(acopy) (Variables|Globals|GetStageShowActors) _returnvalue_9)))
      ;[1 1 1] Advance the index
      (Variables|SetBy-RefVar (Variables|GetIndex) (+ (LocalVariables|GetmyIndex) 1))
      (else
        ;[1 1 1] If we executed all loops, end the animation
        (Variables|SetBy-RefVar (Variables|Globals|GetLoopIndex))))
    (return (LocalVariables|GetmyResult) (and (< (Variables|Globals|GetLoopIndex) (Variables|Globals|GetLoopCount)) (> (Variables|Globals|GetLoopIndex) 0))))
  (goto _label))

(fn GetRandomFx
  (params (Index :type Integer%20%28by%20ref%29) (LoopIndex :type Integer%20%28by%20ref%29))
  (bind _myloopindex (LocalVariables|GetmyLoopIndex))
  (bind _returnvalue (+ _myloopindex 1))
  (LocalVariables|SetmyIndex Index)
  (LocalVariables|SetmyLoopIndex LoopIndex)
  (Utilities|Array|Clear (LocalVariables|GetLocalActor))
  (if (not (< (LocalVariables|GetmyIndex) (Utilities|Array|Length (Variables|Globals|GetStageShowActors))))
    ;[1 1 1] Reset the actor index and count a completed loop. If we completed the set loop count,
    ; we finished this animation
    (LocalVariables|SetmyIndex 0)
    (Variables|SetBy-RefVar (Variables|Globals|GetLoopIndex) _returnvalue)
    (if (not (< _returnvalue (Variables|Globals|GetLoopCount)))
      (Variables|SetBy-RefVar (Variables|Globals|GetLoopIndex))
      (return (LocalVariables|GetLocalActor) (and (< (Variables|Globals|GetLoopIndex) (Variables|Globals|GetLoopCount)) (> (Variables|Globals|GetLoopIndex) 0)))))
  (Utilities|Array|AddUnique (LocalVariables|GetLocalActor) (Utilities|Array|RandomArrayItem (Variables|Globals|GetStageShowActors)))
  ;[1 1 1] Increment index
  (Variables|SetBy-RefVar (Variables|GetIndex) (+ (LocalVariables|GetmyIndex) 1))
  (return (LocalVariables|GetLocalActor) (and (< (Variables|Globals|GetLoopIndex) (Variables|Globals|GetLoopCount)) (> (Variables|Globals|GetLoopIndex) 0))))

(fn GetStageManager
  (bind _stagemanager (Variables|AutonomousLoop|GetStageManager)
    (:then
      (return _stagemanager true))
    (:else
      (bind _returnvalue (Actor|GetActorOfClass "/SparksStageShow/Systems/StageManager/BP_MojoStageManagerBase.BP_MojoStageManagerBase_C"))
      (bind _output_get (Variables|AutonomousLoop|SetStageManager _returnvalue))
      (return _output_get false))))

(fn ListenToBeatEvent
  (params (Enabled :type Boolean))
  (bind (_stagemanager _isvalid) (CallFunction|GetStageManager))
  (bind _outputdelegate (EventDispatchers|CreateEvent :!event StartListeningBeatEvents))
  (bind _output_get (Variables|Default|SetAssignedToBeat Enabled))
  (if _output_get
    (Default|BindEventtoOnBeatCalled _stagemanager _outputdelegate)
    (elif _isvalid
      (Default|UnbindEventfromOnBeatCalled _stagemanager _outputdelegate))))

(fn OnBeat
  ;[1 1 1] Only fires the FX on beat on the set beat frequency
  ; (Each 2 beats, for example)
  (bind _result (Math|Integer|IncrementInt (Variables|AutonomousLoop|GetBeatFrequencyCounter))
    (:%20%20
      (if (>= _result (Variables|/SparksStageShow/Systems/StageManager/FXManager/MojoFxManagerEvent_DataAsset.MojoFxManagerEvent_DataAsset_C|GetBeatFrequency (Variables|Globals|GetCurrentFxEvent)))
        (Variables|AutonomousLoop|SetBeatFrequencyCounter 0)
        ;[1 1 1] This value is set to 0.5 when it's a beat and 1.0 when it's a measure.
        ;
        ; The sum is made to handle simultaneous beat and measure events happening on the same
        ; frame.
        ;
        ; If both happen at the same time, we pass 1.0
        ; and it will count as a measure event
        (Variables|Globals|SetTimeSignatureValue 0.5) ; This value is set to 0.5 when it's a beat\nand 1.0 when it's a measure.\nThe sum is made to handle simultaneous\nbeat and measure events happening on the same\nframe. If both happen at the same time, we pass 1.0\nand it will count as a measure event
        (AmbientFX|FireLoopingEvent)))))

(fn OnMeasure
  ;[1 1 1] Only fires the FX on measure on the set measure frequency
  ; (Each 2 measures, for example)
  (bind _result (Math|Integer|IncrementInt (Variables|AutonomousLoop|GetMeasureFrequencyCounter))
    (:%20%20
      (if (>= _result (Variables|/SparksStageShow/Systems/StageManager/FXManager/MojoFxManagerEvent_DataAsset.MojoFxManagerEvent_DataAsset_C|GetMeasureFrequency (Variables|Globals|GetCurrentFxEvent)))
        (Variables|AutonomousLoop|SetMeasureFrequencyCounter 0)
        ;[1 1 1] This value is set to 0.5 when it's a beat
        ; and 1.0 when it's a measure.
        ; The sum is made to handle simultaneous
        ; beat and measure events happening on the same
        ; frame. If both happen at the same time, we pass 1.0
        ; and it will count as a measure event
        (Variables|Globals|SetTimeSignatureValue 1.0) ; This value is set to 0.5 when it's a beat\nand 1.0 when it's a measure.\nThe sum is made to handle simultaneous\nbeat and measure events happening on the same\nframe. If both happen at the same time, we pass 1.0\nand it will count as a measure event
        (AmbientFX|FireLoopingEvent)))))

(fn PlayFXArraybyIndices
  (params (ActorArray :type Array%20of%20Actor%20Object%20References) (ArrayIndices :type Array%20of%20Integers))
  ;[1 1 1] Seems unused, the StageShowFxInterface usually handles this
  (for _array_element ArrayIndices
    (Default|PlayFXOnComponents (Utilities|Array|Get(acopy) ActorArray _array_element))))

(fn PlayNextFX
  (switch Utilities|FlowControl|Switch|SwitchonEShowFXAnimation (Variables|Globals|GetAnimationStyle)
    (:FireAll
      ;[1 1 1] These animation style functions pick actors managed by this FX Manager to fire an FX. For example, the "Linear" anim style will make the actors fire the FX one by one on a linear fashion.
      ;
      ; To acomplish this, the PlayNextFx function will execute several times, going through the steps of the animation and these functions here return the actors for this step.
      ;
      ; The LoopIndex and FxIndex holds the animation state.
      ;
      ; The animation ends when there's no eligible actors. (Actors array is empty)
      (bind _fxarray (Animations|GetNextFxFireAll (Variables|Globals|GetLoopIndex)))
      (LocalVariables|SetActors _fxarray)
      (LocalVariables|SetLocalWillLoop false))
    (:FireIndices
      (bind (_fxarray_2 _willloop) (Animations|GetNextFxFireIndices (Variables|Globals|GetLoopIndex)))
      (LocalVariables|SetActors _fxarray_2)
      (LocalVariables|SetLocalWillLoop _willloop))
    (:Linear
      (bind (_fxarray_3 _willloop_2) (Animations|GetNextFxLinear (Variables|Globals|GetFxIndex) (Variables|Globals|GetLoopIndex)))
      (LocalVariables|SetActors _fxarray_3)
      (LocalVariables|SetLocalWillLoop _willloop_2))
    (:Alternating
      (bind (_fxarray_4 _willloop_3) (Animations|GetNextFxAlternate (Variables|Globals|GetFxIndex) (Variables|Globals|GetLoopIndex)))
      (LocalVariables|SetActors _fxarray_4)
      (LocalVariables|SetLocalWillLoop _willloop_3))
    (:OutsideIn
      (bind (_fxarray_5 _willloop_4) (Animations|GetNextFxOutsideinOut (Variables|Globals|GetFxIndex) (Variables|Globals|GetLoopIndex)))
      (LocalVariables|SetActors _fxarray_5)
      (LocalVariables|SetLocalWillLoop _willloop_4))
    (:Random
      (bind (_actors _willloop_5) (Animations|GetRandomFx (Variables|Globals|GetFxIndex) (Variables|Globals|GetLoopIndex)))
      (LocalVariables|SetActors _actors)
      (LocalVariables|SetLocalWillLoop _willloop_5)))
  (if (Utilities|Array|IsNotEmpty (LocalVariables|GetActors))
    ;[1 1 1] If the animation has eligible actors, execute PlayFX on them and schedule another run through
    ; PlayNextFx to process the next anim step. The next step will wait for what is defined on the Delay variable.
    (CallFunction|PlayFXArray (LocalVariables|GetActors))
    (bind _returnvalue (Utilities|Time|SetTimerbyEvent (EventDispatchers|CreateEvent :!event Play%20Next%20FX) (Math|Float|Max(Float) (Variables|Globals|GetDelay) 0.01)))
    (Variables|Globals|SetPlayNextFXTimerHandle _returnvalue)
    (else
      ;[1 1 1] If the animation doesn't have any eligible actors, it means it's finished.
      (Animations|StopandResetFxManagerState)
      (Default|CallEventFinished self)
      (if (Variables|AutonomousLoop|GetAutonomousLoop)
        (AmbientFX|GetNextAutonomousEvent)))))

(fn PlayFXArray
  (params (FXArray :type Array%20of%20Actor%20Object%20References))
  (LocalVariables|SetLocalFXArray FXArray)
  ;[1 1 1] Here we call the functions from StageShowFXInterface and pass some parameters that Actors that implement it can use for different VFXs
  (switch Utilities|FlowControl|Switch|SwitchonEShowFXType (Variables|Globals|GetType)
    (:Impulse
      (for _array_element (LocalVariables|GetLocalFXArray)
        (Class|StageShowFXInterface|MojoStageShowImpulse(Message) _array_element (Variables|Globals|GetTimeSignatureValue))))
    (:Activate
      (for _array_element_2 (LocalVariables|GetLocalFXArray)
        (Class|StageShowFXInterface|MojoStageShowActivate(Message) _array_element_2 1.0 (Variables|Globals|GetTimeSignatureValue))))
    (:Deactivate
      (for _array_element_3 (LocalVariables|GetLocalFXArray)
        (Class|StageShowFXInterface|MojoStageShowActivate(Message) _array_element_3 0.0 (Variables|Globals|GetTimeSignatureValue))))
    (:ActivateBeat
      (for _array_element_4 (LocalVariables|GetLocalFXArray)
        (Class|StageShowFXInterface|MojoStageShowActivateBeat(Message) _array_element_4 1.0 (Variables|Globals|GetTimeSignatureValue))))
    (:DeactivateBeat
      (for _array_element_5 (LocalVariables|GetLocalFXArray)
        (Class|StageShowFXInterface|MojoStageShowActivateBeat(Message) _array_element_5 0.0 (Variables|Globals|GetTimeSignatureValue))))))

(fn PlayFXOnComponents
  (params (Actor :type Actor%20Object%20Reference))
  ;[1 1 1] Seems unused, the FxManager never fires Niagara FXs directly like this
  (for _array_element (Actor|GetComponentsByClass Actor "/Script/Niagara.NiagaraComponent")
    (Components|Activation|Activate _array_element false)))

(fn SetGlobalVars
  (params (FXManagerEvent :type Mojo%20Fx%20Manager%20Event%20Data%20Asset%20Object%20Reference%20%28by%20ref%29))
  (bind _output (Utilities|ResolveSoftReference (Variables|/SparksStageShow/Systems/StageManager/FXManager/MojoFxManagerEvent_DataAsset.MojoFxManagerEvent_DataAsset_C|GetNiagaraSystemOverride FXManagerEvent)))
  (bind _niagarasystemoverride (Variables|/SparksStageShow/Systems/StageManager/FXManager/MojoFxManagerEvent_DataAsset.MojoFxManagerEvent_DataAsset_C|GetNiagaraSystemOverride FXManagerEvent))
  (if (Utilities|IsValid FXManagerEvent :!class /Script/Engine.KismetSystemLibrary)
    ;[1 1 1] Setting animation variables
    (Variables|Globals|SetType (Utilities|Enum|BytetoEnumEShowFXType (Math|Conversions|ToByte(Integer) (Math|Conversions|ToInteger(Byte) (Variables|/SparksStageShow/Systems/StageManager/FXManager/MojoFxManagerEvent_DataAsset.MojoFxManagerEvent_DataAsset_C|GetFxType FXManagerEvent)))))
    (Variables|Globals|SetAnimationStyle (Utilities|Enum|BytetoEnumEShowFXAnimation (Math|Conversions|ToByte(Integer) (Math|Conversions|ToInteger(Byte) (Variables|/SparksStageShow/Systems/StageManager/FXManager/MojoFxManagerEvent_DataAsset.MojoFxManagerEvent_DataAsset_C|GetFxAnimation FXManagerEvent)))))
    (Variables|Globals|SetLoopCount (Variables|/SparksStageShow/Systems/StageManager/FXManager/MojoFxManagerEvent_DataAsset.MojoFxManagerEvent_DataAsset_C|GetLoopCount FXManagerEvent))
    (Variables|Globals|SetDelay (Variables|/SparksStageShow/Systems/StageManager/FXManager/MojoFxManagerEvent_DataAsset.MojoFxManagerEvent_DataAsset_C|GetAnimationStepDelay FXManagerEvent))
    (Variables|Globals|SetAlternateCount (Variables|/SparksStageShow/Systems/StageManager/FXManager/MojoFxManagerEvent_DataAsset.MojoFxManagerEvent_DataAsset_C|GetAlternateCount FXManagerEvent))
    (Variables|Globals|SetIndices (Variables|/SparksStageShow/Systems/StageManager/FXManager/MojoFxManagerEvent_DataAsset.MojoFxManagerEvent_DataAsset_C|GetIndices FXManagerEvent))
    (Variables|Globals|SetReverse (Variables|/SparksStageShow/Systems/StageManager/FXManager/MojoFxManagerEvent_DataAsset.MojoFxManagerEvent_DataAsset_C|GetReverse FXManagerEvent))
    ;[1 1 1] Check if a Niagara change is set on this event
    (if (Utilities|IsValidSoftObjectReference _niagarasystemoverride)
      ;[1 1 1] Check if the soft reference loaded correctly
      (Utilities|IsValid _output
        (:Is%20Valid
          ;[1 1 1] If everything is ok, send the Update Niagara signal to the managed actors
          (CallFunction|UpdateActorsNiagaraSystem _output))
        (:Is%20Not%20Valid
          (Fortnite|FortShippingLog self "[FxManager] Event Niagara failed to load." true))))))

(fn StartAutonomousLoopEvent)

(fn StartDebugAmbientFx
  (AmbientFX|ValidateAutonomousLoopEvents)
  (Variables|Globals|SetCurrentFxEvent (Utilities|Array|Get(acopy) (Variables|AutonomousLoop|GetValidAutonomousLoopEvents)))
  (AmbientFX|StartLoopingEventbyTimer))

(fn StartLoopingEventByTimer
  ;[1 1 1] Only starts a new AmbientFX loop if frequency is not zero and there's no other ambientFX loop running
  (if (and (> (Variables|/Script/SparksStageShowRuntime.StageShowEvent|GetNoteFrequency (Variables|Globals|GetCurrentFxEvent)) 0) (not (Utilities|Time|IsValidTimerHandle (Variables|AutonomousLoop|GetAutonomousLoopTimer))))
    ;[1 1 1] Fire the ambient FX and schedule it to repeat every X seconds (set on Ambient Fx Frequency)
    (AmbientFX|FireLoopingEvent)
    (bind _returnvalue (Utilities|Time|SetTimerbyEvent (EventDispatchers|CreateEvent :!event FireLoopingEvent) (Variables|/Script/SparksStageShowRuntime.StageShowEvent|GetNoteFrequency (Variables|Globals|GetCurrentFxEvent)) true))
    (Variables|AutonomousLoop|SetAutonomousLoopTimer _returnvalue)))

(fn StopAndResetFxManagerState
  ;[1 0.8004 0.2016] This function terminates an animation. It also can stop a animation midway.
  ; It reset the variables, mark this FxManager as not busy anymore and call the EventFinished dispatcher.
  ; If this event was fired by the stage manager, the spawned event will be listening to this dispatcher, signaling to it that we're done.
  ; (The CheckIfEventIsComplete function from BP_MojoStageEvent is binded here)
  (Utilities|Time|ClearandInvalidateTimerbyHandle (Variables|Globals|GetPlayNextFXTimerHandle))
  (Variables|Globals|SetFxIndex 0)
  (Variables|Globals|SetLoopIndex 0)
  (Variables|Globals|SetTimeSignatureValue 0.0)
  (Variables|Globals|SetIsBusy false))

(fn StopDebugAmbientFx
  (AmbientFX|StopLoopingEvent))

(fn StopEventAndFinish
  (Animations|StopandResetFxManagerState)
  (AmbientFX|StopLoopingEvent)
  (CallFunction|ListentoBeatEvent false)
  (for _array_element (Variables|Globals|GetStageShowActors)
    (Class|StageShowFXInterface|MojoStageShowonStop(Message) _array_element))
  (Default|CallEventFinished self))

(fn StopLoopingEvent
  (bind _onbeatasyncaction (Variables|AutonomousLoop|GetOnBeatAsyncAction))
  (Utilities|FlowControl|Sequence
    (:then_0
      ;[1 1 1] Stop ambient FX timer
      (if (Utilities|Time|IsValidTimerHandle (Variables|AutonomousLoop|GetAutonomousLoopTimer))
        (Utilities|Time|ClearandInvalidateTimerbyHandle (Variables|AutonomousLoop|GetAutonomousLoopTimer))))
    (:then_1
      ;[1 1 1] Stop listening to OnBeat event
      (Utilities|IsValid _onbeatasyncaction
        (:Is%20Valid
          (GameplayEventRouter|StopListeningtoEvent _onbeatasyncaction))))
    (:then_2
      (Variables|Globals|SetisPlayingLoopingEvent false))))

(fn UpdateActorsNiagaraSystem
  (params (NiagaraSystem :type Niagara%20System%20Object%20Reference))
  (if (!= (Variables|Globals|GetNiagaraSystem) NiagaraSystem)
    (Utilities|IsValid NiagaraSystem
      (:Is%20Valid
        (Variables|Globals|SetNiagaraSystem NiagaraSystem)
        (for _array_element (Variables|Globals|GetStageShowActors)
          (Class|StageShowFXInterface|MojoStageShowUpdateNiagaraSystem(Message) _array_element (Variables|Globals|GetNiagaraSystem))))
      (:Is%20Not%20Valid
        (Fortnite|FortShippingLog self "[FxManager] Failed to send a Niagara System update. Niagara System is invalid." true)))))

(fn ConstructionScript)

(fn ValidateAutonomousLoopEvents
  ;[1 1 1] Creates an Autonomous Loop Events array containing only valid FxEvents. Returns invalid if there's no valid events
  (if (Utilities|Array|IsNotEmpty (Variables|AutonomousLoop|GetAutonomousLoopEvents))
    (Utilities|Array|Clear (Variables|AutonomousLoop|GetValidAutonomousLoopEvents))
    (for _array_element (Variables|AutonomousLoop|GetAutonomousLoopEvents)
      (bind _asmojo_fx_manager_event_data_asset (Utilities|Casting|CastToMojoFxManagerEvent_DataAsset _array_element))
      (if (Utilities|IsValid _asmojo_fx_manager_event_data_asset :!class /Script/Engine.KismetSystemLibrary)
        (Utilities|Array|Add (Variables|AutonomousLoop|GetValidAutonomousLoopEvents) _asmojo_fx_manager_event_data_asset)))
    (return (Utilities|Array|IsNotEmpty (Variables|AutonomousLoop|GetValidAutonomousLoopEvents)))
    (else
      (return false))))
